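# Translucent unlit master material for holographic Niagara sprite and mesh particles, stored with the AshenMagnusSlant skydiving contrail.
# A static or flipbook RGB texture, with optional rotation and noise-distorted UVs, supplies channel-weighted shapes tinted by particle color.
# World-height scanlines, scaled by camera distance, and flickering world-aligned noise modulate depth-faded opacity, while a procedural radial radar sweep adds emissive.
# Dynamic parameters drive animation rate, scanline tiling, flicker and randomness; low quality zeroes flicker noise.
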Material /BRCosmetics/Athena/Effects/Particles/Skydiving/NS_Contrail_AshenMagnusSlant/Materials/M_FX_HoloMaster
graph {
 properties {
  bUsedWithNiagaraMeshParticles=true
  bUsedWithNiagaraSprites=true
  blendMode=BLEND_Translucent
  pixelDepthOffsetMode=PDOM_Legacy
  shadingModel=MSM_Unlit
  twoSided=true
  usedShadingModels=Unlit
 }
 nodes [
  n1 = ParticleColor()
  n2 = NamedRerouteUsage()
  n3 = ScalarParameter() {defaultValue=1, parameterName=Scale U}
  n4 = ScalarParameter() {defaultValue=1, parameterName=Scale V}
  n5 = MaterialFunctionCall(Float3=Saturate(None=NamedRerouteUsage())) {materialFunction=&BreakOutFloat3Components}
  n6 = Subtract(A=@n5.R, B=@n5.R)
  n7 = AppendVector(A=Multiply(A=@n6, B=@n6), B=Constant())
  n8 = NamedRerouteUsage()
  n9 = Add(
   A=Ceil(None=Saturate(None=@n8))
   B=DynamicParameter() {defaultValue=<rgba 1 1 0 1>, paramNames [Animation FPS, ScanlineTileMult, Flicker, Random]}.Flicker
  )
  n10 = Multiply(A=Time(), B=ScalarParameter() {defaultValue=1, parameterName=Rotator Speed})
  n11 = Multiply(A=Sine(None=TextureSample(UVs=@n10) {samplerType=SAMPLERTYPE_Grayscale, texture=/Game/Effects/Fort_Effects/Textures/Fire/T_WispyFireNoise_Soft_M}.R)) {constB=5}
  n12 = Add(
   A=Multiply(A=Multiply(A=@n7, B=@n9), B=ScalarParameter() {defaultValue=0.15, parameterName=NoiseDistortionAmount})
   B=LinearInterpolate(A=TextureCoordinate(), B=Rotator(Time=@n11), Alpha=ScalarParameter() {parameterName=UseRotation})
  )
  n13 = VectorParameter() {
   desc="X and Y should specify the desired center of scaling in pixels. Z should be the texture size in pixels (ie, 1024). Currently only supports square textures."
   parameterName=Scale Offset
   defaultValue=<rgba 0 0 1024 1>
  }
  n14 = TextureObjectParameter() {
   parameterName=Texture
   samplerType=SAMPLERTYPE_LinearColor
   texture=/BRCosmetics/Athena/Effects/Particles/Skydiving/NS_Contrail_AshenMagnusSlant/Textures/T_FX_RadarRGB_01
  }
  n15 = TextureSample(
   UVs=MaterialFunctionCall!(
    Texture Scale=MaterialFunctionCall(R=@n3, G=@n4) {materialFunction=&AppendMany}.RG
    UVs=@n12
    Offset XY Texturesize Z=@n13.RGB
   ) {materialFunction=/Engine/Functions/Engine_MaterialFunctions01/Texturing/ScaleUVsAroundPoint}.UVs
   Tex=@n14
  ) {
   samplerType=SAMPLERTYPE_LinearColor
   texture=&T_FX_RadarRGB_01
  }
  n16 = Multiply(A=Time(), B=DynamicParameter() {defaultValue=<rgba 1 1 0 1>, paramNames [Animation FPS, ScanlineTileMult, Flicker, Random]}.Animation FPS)
  n17 = ScalarParameter() {defaultValue=1, parameterName=NumRows}
  n18 = ScalarParameter() {defaultValue=1, parameterName=NumColumns}
  n19 = StaticBoolParameter() {parameterName=MipBias}
  n20 = StaticComponentMaskParameter(None=LinearInterpolate(
    A=@n15.RGB
    B=MaterialFunctionCall!(
     "Animation  Phase (0-1)"=@n16
     Number of Rows=@n17
     Number of Columns=@n18
     Texture=@n14
     UVs=@n12
     MipBias/Level
     "Use Mip Bias (T) Level (F)"=@n19
     "Clamp Anim (see tooltip)"
    ) {materialFunction=/BRCosmetics/Characters/Player/Female/Medium/Bodies/F_MED_ZombieElasticNeon/FX/FlipBook_Clamped}.Result
    Alpha=ScalarParameter() {defaultValue=1, parameterName=UseFlipbook}
   )) {defaultB=true, defaultG=true, defaultR=true, parameterName=RGB Mask}
  n21 = Multiply(A=ScalarParameter() {defaultValue=1, parameterName=Red Channel Float}, B=ChannelMaskParameter(None=@n20) {parameterName=Red Channel}.R)
  n22 = Multiply(
   A=ScalarParameter() {defaultValue=1, parameterName=Green Channel Float}
   B=ChannelMaskParameter(None=@n20) {maskChannel=Green, parameterName=Green Channel, defaultValue=<rgba 0 1 0 0>}.R
  )
  n23 = Multiply(
   A=ScalarParameter() {defaultValue=1, parameterName=Blue Channel Float}
   B=ChannelMaskParameter(None=@n20) {maskChannel=Blue, parameterName=Blue Channel, defaultValue=<rgba 0 0 1 0>}.R
  )
  n24 = Multiply(
   A=Add(A=Add(A=@n21, B=@n22), B=@n23)
   B=LinearInterpolate(
    A=ScalarParameter() {defaultValue=1, parameterName=EmissiveLerpMin}
    B=ScalarParameter() {defaultValue=0.4, parameterName=EmissiveLerpMax}
    Alpha=@n23
   ) {constA=1, constB=0.25}
  )
  n25 = Multiply(A=@n24, B=Multiply(A=ScalarParameter() {defaultValue=-1, group=Noise, parameterName=NoiseBrightness}, B=@n2) {constB=0.1})
  n26 = Multiply(A=@n2) {constB=1.5}
  n27 = Clamp(None=Subtract(A=ScalarParameter() {defaultValue=0.4, parameterName=MinScanlineOpacity}, B=@n26)) {minDefault=-0.25}
  n28 = Add(A=@n26)
  n29 = Multiply(
   A=DynamicParameter() {defaultValue=<rgba 1 1 0 1>, paramNames [Animation FPS, ScanlineTileMult, Flicker, Random]}.ScanlineTileMult
   B=ScalarParameter() {defaultValue=300, parameterName=ScanlineTile}
  )
  n30 = NamedRerouteUsage()
  n31 = LinearInterpolate(Alpha=Ceil(None=Saturate(None=@n30))) {constA=1, constB=0.7}
  n32 = ComponentMask(None=WorldPosition().XYZ) {b=true}
  n33 = Distance(A=CameraPositionWS(), B=ActorPositionWS())
  n34 = ScalarParameter() {defaultValue=20000, parameterName=ScaleDistance}
  n35 = LinearInterpolate(B=@n34, Alpha=Saturate(None=Divide(A=@n33, B=@n34) {constB=1000})) {constA=50, constB=100}
  n36 = Sine(None=Multiply(A=Multiply(A=@n29, B=@n31), B=Divide(A=@n32, B=@n35) {constB=20}) {constA=1})
  n37 = Abs(None=DotProduct(A=VertexNormalWS(), B=Constant3Vector() {constant=<rgba 0 0 1 0>}))
  n38 = Power(Base=LinearInterpolate(A=ConstantBiasScale(None=@n36), Alpha=Power(Base=@n37) {constExponent=5}) {constB=0.25}) {constExponent=1.5}
  n39 = Saturate(None=Multiply(A=Add(A=@n24, B=@n25), B=LinearInterpolate(A=@n27, B=@n28, Alpha=@n38) {constA=0.5}))
  Add(
   A=DepthFade(Opacity=Multiply(A=@n1.A, B=@n39), FadeDistance=ScalarParameter() {defaultValue=5, parameterName=FadeDistance})
   B=MaterialFunctionCall() {materialFunction=&RadialGradientExponential}.RadialGradientExponential
  )
  n40 = TextureCoordinate()
  n41 = Distance(A=Multiply(A=Subtract(A=@n40) {constB=0.5}) {constB=2}, B=Constant2Vector())
  n42 = MaterialFunctionCall() {materialFunction=&VectorToRadialValue}
  n43 = Time()
  n44 = ScalarParameter() {defaultValue=-1, parameterName=Radar Velocity}
  n45 = Frac(None=Add(A=ComponentMask(None=@n42.Radial Coordinates) {r=true}, B=Multiply(A=@n43, B=@n44) {constB=-1}))
  n46 = Add(A=Multiply(A=Power(Base=@n45) {constExponent=14}) {constB=5}, B=Multiply(A=Power(Base=@n45) {constExponent=2000}) {constB=500})
  n47 = Multiply(
   A=OneMinus(None=SmoothStep(Value=@n41) {constMax=0.85})
   B=Multiply(
    A=ScalarParameter() {parameterName=Alpha Channel Float}
    B=ChannelMaskParameter(None=@n46) {maskChannel=Alpha, parameterName=Alpha Channel, defaultValue=<rgba 0 0 0 1>}.R
   )
  )
  MaterialFunctionCall(Emissive=Add(A=Multiply(A=@n47, B=@n1.RGB), B=Multiply(A=@n24, B=@n1.RGB) {constB=2})) {materialFunction=&MF_FX_EmissiveAdjustments}
  n48 = VectorParameter() {group=Noise, parameterName=NoiseTileMult, defaultValue=<rgba 1.5 1.5 0.06 1>}
  n49 = DynamicParameter() {parameterIndex=1, defaultValue=<rgba 1 1 0 1>, paramNames [NoiseTileMult, Param2, Param3, Param4]}
  n50 = Divide(A=ObjectRadius(), B=Multiply(A=@n49.NoiseTileMult, B=ScalarParameter() {defaultValue=0.25, group=Noise, parameterName=NoiseSize}))
  n51 = WorldPosition()
  n52 = Add(
   A=Add(A=DynamicParameter() {defaultValue=<rgba 1 1 0 1>, paramNames [Animation FPS, ScanlineTileMult, Flicker, Random]}.Flicker, B=Time())
   B=NamedRerouteUsage()
  )
  n53 = Multiply(A=@n50, B=Multiply(A=@n52, B=VectorParameter() {group=Noise, parameterName=NoisePanSpeed, defaultValue=<rgba -0.08 0.07 0.15 0>}.RGB))
  n54 = Power(Base=MaterialFunctionCall(
    TextureObject=TextureObject() {samplerType=SAMPLERTYPE_LinearColor, texture=&T_NoisePack_LinearColor}
    TextureSize=Multiply(A=@n48.RGB, B=@n50)
    WorldPosition=Add(A=@n51.XYZ, B=@n53)
   ) {materialFunction=&WorldAlignedTexture}.XY Texture) {constExponent=4}
  n55 = Multiply(A=@n54, B=ScalarParameter() {defaultValue=2, group=Noise, parameterName=NoiseMaskMult}) {constB=20}
  n56 = DynamicParameter() {defaultValue=<rgba 1 1 0 1>, paramNames [Animation FPS, ScanlineTileMult, Flicker, Random]}
  n57 = Multiply(A=ScalarParameter() {defaultValue=2, parameterName=FlickerSpeed}, B=@n52)
  n58 = AppendVector(A=AppendVector(A=Sine(None=@n57) {period=1.6}, B=Sine(None=@n57) {period=1.42}), B=Sine(None=@n57) {period=1.51})
  n59 = Ceil(None=@n58)
  n60 = ScalarParameter() {defaultValue=0.5, group=Noise, parameterName=NoiseFrequency}
  n61 = Add(A=Sine(None=@n57) {period=0.72}, B=Multiply(B=@n60) {constA=0.75}) {constB=0.75}
  n62 = Multiply(A=Add(A=Sine(None=@n57) {period=2.1}, B=Multiply(B=@n60) {constA=0.25}), B=@n61)
  n63 = Multiply(B=@n60) {constA=0.9, constB=0.9}
  n64 = Sine(None=@n57) {period=1.25}
  n65 = Multiply(A=@n61, B=Ceil(None=Add(A=@n63, B=@n64) {constB=0.9}))
  n66 = Sine(None=@n52) {period=0.075}
  n67 = Multiply(A=Saturate(None=Multiply(A=@n62, B=@n65)), B=LinearInterpolate(Alpha=ConstantBiasScale(None=@n66)) {constA=0.7})
  n68 = Add(A=@n56.Flicker, B=Multiply(A=Multiply(A=@n59, B=@n67), B=ScalarParameter() {defaultValue=2, group=Noise, parameterName=NoiseMult}))
  n69 = Desaturation(None=Saturate(None=Multiply(A=@n55, B=@n68)))
  n70 = ScalarParameter() {defaultValue=0.2, group=Noise, parameterName=NoiseContrast}
  NamedRerouteDeclaration(None=QualitySwitch(
    Default=MaterialFunctionCall(In=@n69, Contrast=@n70) {materialFunction=&CheapContrast}.Result
    Low=Constant()
   )) {name=FlickerNoise, nodeColor=<rgba 0.0312 0 1 1>}
  NamedRerouteDeclaration(None=@n56.Random) {name=Random, nodeColor=<rgba 0.7578 0 1 1>}
  NamedRerouteDeclaration(None=Multiply(A=@n54, B=Add(A=@n58))) {name=NoiseDistortion, nodeColor=<rgba 0.9219 0 1 1>}
  Time()
 ]
}
