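# Additive unlit material for Niagara mesh particles in the AshenMagnusSlant skydiving contrail, drawing a glowing laser-sight style effect.
# A vertical UV gradient, inverted by default through a static switch, is shaped by exponent and scale, clamped, then tinted by particle color.
# The result is multiplied by a Fresnel function, softened at intersections with DepthFade, and passed into a shared emissive adjustment function.
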Material /BRCosmetics/Athena/Effects/Particles/Skydiving/NS_Contrail_AshenMagnusSlant/Materials/M_FX_LaserSight_01
graph {
 properties {
  bAutomaticallySetUsageInEditor=false
  bEnableResponsiveAA=true
  bUsedWithNiagaraMeshParticles=true
  blendMode=BLEND_Additive
  pixelDepthOffsetMode=PDOM_Legacy
  shadingModel=MSM_Unlit
  usedShadingModels=Unlit
 }
 nodes [
  n1 = MaterialFunctionCall(Float2=TextureCoordinate()) {materialFunction=&BreakOutFloat2Components}
  n2 = Power(
   Base=StaticSwitchParameter(True=OneMinus(None=@n1.G), False=@n1.G) {defaultValue=true, parameterName=InvertGradient}
   Exp=ScalarParameter() {defaultValue=2, parameterName=Gradient Exponent}
  )
  n3 = ScalarParameter() {defaultValue=1, parameterName=GradientScale}
  n4 = Multiply(A=ParticleColor().RGB, B=Clamp(None=Multiply(A=@n2, B=@n3)))
  n5 = MaterialFunctionCall(
   Brightness=ScalarParameter() {defaultValue=2, parameterName=FresnelBrightness}
   Exponent=ScalarParameter() {defaultValue=2, parameterName=FresnelExponent}
  ) {materialFunction=&MF_Fresnel_FX}
  MaterialFunctionCall(Emissive=DepthFade(Opacity=Multiply(A=@n4, B=@n5.Fresnel), FadeDistance=ScalarParameter() {defaultValue=5, parameterName=DepthFadeDistance})) {materialFunction=&MF_FX_EmissiveAdjustments}
 ]
}
