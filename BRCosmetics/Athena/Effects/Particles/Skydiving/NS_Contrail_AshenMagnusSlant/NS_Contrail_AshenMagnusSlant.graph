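# Niagara skydiving contrail cosmetic that draws a targeting rig on the player: radar sprite, three HUD meshes, two gadget meshes, two lasers and a pulsing aim reticle.
# Gadgets and lasers follow skeletal mesh bones or sockets, with lasers smoothly turning toward a lean-driven point.
# ParaGlideLeanAlpha also shifts the reticle, blends laser and reticle colors from green to red, and drives HUD material parameters.
# Front-end flags adjust HUD color, rotation and material.
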NiagaraSystem /BRCosmetics/Athena/Effects/Particles/Skydiving/NS_Contrail_AshenMagnusSlant/NS_Contrail_AshenMagnusSlant
dependencies {
 usedDataInterfaces [
  &NiagaraDataInterfaceArrayFloat3
  &NiagaraDataInterfaceArrayFloat2
  &NiagaraDataInterfaceColorCurve
  &NiagaraDataInterfaceSkeletalMesh
 ]
 usedDynamicInputs [&SelectVector2DFromArray]
 usedModules [
  &self:SystemSpawnScript.SystemScriptSource.SystemScriptGraph.NiagaraNodeAssignment_0.SetVariables_#1
  &SystemState
  /Game/Effects/Niagara/Modules/Contrails/FN_Contrail_Setup
  &FN_Integrator
  &self:SystemSpawnScript.SystemScriptSource.SystemScriptGraph.NiagaraNodeAssignment_1.SetVariables_#2
 ]
}
emitters {
 LWE_Targets {rendererClasses [], renderers []}
 NE_Aim {
  emitterUpdateScript [
   &EmitterState(Life Cycle Mode=::NewEnumerator1, Loop Duration=0.1, UseLoopDelay=false)
   &SpawnRate(SpawnRate=3, Use Spawn Probability=false)
  ]
  particleSpawnScript [
   /SetVariables(Particles.Linear Color=<rgba 0 5 0.0568 1>) {isSetParameters=true}
   &InitializeParticle(
    Color=<rgba 5 0 0.0894 1>
    Lifetime=0.35
    Position Offset=&Lerp_Vector(A=Linked() {linkedVariable {name=System.Position, type=NiagaraPosition}}, B=<1 40 0>, Alpha=$User.ParaGlideLeanAlpha)
    Position Offset Coordinate Space=::Local
    Sprite Size Mode=::NewEnumerator3
    Uniform Sprite Size=35
    UsePositionOffset=true
    Write Material Random=true
   )
   &InitialMeshOrientation(Mesh Orientation Mode=::NewEnumerator2, Rotation=<0 0.25 0>, Rotation Coordinate Space=::NewEnumerator0, Use Rotation Vector=true)
   &ApplyInitialForces()
   &AlignSpriteToMeshOrientation(SetAlignment=true, SetFacing=true) {module=AlignSpriteToMeshOrientation001}
   &CameraOffset(Camera Offset Amount=&FN_FrontEndFloatModifier(Gameplay=0, Front End=5))
  ]
  particleUpdateScript [
   &ParticleState()
   &SolveForcesAndVelocity(Clamp Velocity=false, Limit Acceleration=false)
   &Color(Color=&LerpLinearColors(
     StartColor=<rgba 1 0 0.0179 1>
     EndColor=&ColorFromCurve(ColorCurve=DataInterface() {propertyValues {
        AlphaCurve {defaultValue=1, keys [<at 0 1 RCIM_Cubic>,<at 1 0 RCIM_Cubic>]}
        BlueCurve {keys [<at 0 0 RCIM_Cubic>,<at 1 0 RCIM_Cubic>]}
        GreenCurve {keys [<at 0 0 RCIM_Cubic>,<at 1 0 RCIM_Cubic>]}
        RedCurve {keys [<at 0 25 RCIM_Cubic>,<at 1 0 RCIM_Cubic>]}
       }})
     LerpFactor=$User.ParaGlideLeanAlpha
    ))
   /SetVariables(Particles.Linear Color=&LerpLinearColors(
     StartColor=<rgba 0 1 0.0114 1>
     EndColor=&ColorFromCurve(ColorCurve=DataInterface() {propertyValues {
        AlphaCurve {keys [<at 0 1 RCIM_Cubic>,<at 1 0 RCIM_Cubic>]}
        BlueCurve {keys [<at 0 0 RCIM_Cubic>,<at 1 0 RCIM_Cubic>]}
        GreenCurve {keys [<at 0 0 RCIM_Cubic>,<at 1 0 RCIM_Cubic>]}
        RedCurve {keys [<at 0 25 RCIM_Cubic>,<at 1 0 RCIM_Cubic>]}
       }})
     LerpFactor=$User.ParaGlideLeanAlpha
    )) {isSetParameters=true}
  ]
  rendererClasses [&NiagaraSpriteRendererProperties]
  renderers [
   {rendererClass=&NiagaraSpriteRendererProperties}
   {rendererClass=&NiagaraSpriteRendererProperties, rendererIndex=1}
  ]
 }
 NE_BaseRadar {
  emitterUpdateScript [
   &EmitterState()
   &SpawnBurst_Instantaneous(Spawn Count=1, Spawn Probability=HlslExpression() {hlslExpression="Emitter.LoopCount == 0"}, Use Spawn Probability=true)
  ]
  particleSpawnScript [
   &InitializeParticle(
    Color=<rgba 0.1177 1.25 0 1>
    Lifetime Max=1
    Lifetime Mode=::NewEnumerator1
    Sprite Size Mode=::NewEnumerator3
    Uniform Sprite Size=250
    UsePositionOffset=false
    Write Material Random=true
   )
   &InitialMeshOrientation(Mesh Orientation Mode=::NewEnumerator2, Rotation=<0 -0.25 0>, Rotation Coordinate Space=::NewEnumerator2, Use Rotation Vector=true)
   &ApplyInitialForces()
   &AlignSpriteToMeshOrientation(SetAlignment=true, SetFacing=true)
   &CameraOffset(Camera Offset Amount=&FN_FrontEndFloatModifier(Gameplay=0, Front End=5))
  ]
  particleUpdateScript [
   &ParticleState(
    Kill Particles When Lifetime Has Elapsed=false
    Let Infinitely Lived Particles Die When Emitter Deactivates=true
    Loop Particles Lifetime=true
   )
   &SolveForcesAndVelocity(Clamp Velocity=false, Limit Acceleration=false)
   &DynamicMaterialParameters(Write Parameter Index 0=::NewEnumerator0)
  ]
  rendererClasses [&NiagaraSpriteRendererProperties]
  renderers [
   {rendererClass=&NiagaraSpriteRendererProperties}
   {rendererClass=&NiagaraSpriteRendererProperties, rendererIndex=1}
  ]
 }
 NE_Gadget {
  emitterUpdateScript [
   &EmitterState()
   &SpawnBurst_Instantaneous(
    Spawn Count=2
    Spawn Probability=HlslExpression() {hlslExpression="Emitter.LoopCount == 0"}
    Use Spawn Probability=true
    UseLoopCountLimit=false
   )
  ]
  particleSpawnScript [
   &PartitionParticles()
   &InitializeParticle(Mesh Scale Mode=::NewEnumerator3, Mesh Uniform Scale=0.8, UsePositionOffset=false, Write Material Random=true)
   &InitialMeshOrientation(
    Mesh Axis To Orient=<0 1 0>
    Mesh Orientation Mode=::NewEnumerator2
    Orientation Coordinate Space=::Local
    Orientation Vector=<0 0 0>
    Rotation=&SelectVectorFromArray(
     Direct Array Index=$Particles.PartitionParticles.Partition
     Vector Selection Array=DataInterface() {propertyValues {FloatData [<-0.25 0.04 0>, <0.25 0 0>]}}
    )
    Rotation Coordinate Space=::NewEnumerator2
    Use Rotation Vector=true
   )
  ]
  particleUpdateScript [
   &ParticleState(Kill Particles When Lifetime Has Elapsed=false, Let Infinitely Lived Particles Die When Emitter Deactivates=true)
   &SolveForcesAndVelocity(Clamp Velocity=false, Limit Acceleration=false)
   &SkeletalMeshLocation(
    Bone / Socket Index=$Particles.PartitionParticles.Partition
    MaskSpawnGroup=false
    Mesh Sampling Type=::NewEnumerator5
    Skeletal Mesh=$System.Skeletal Mesh
    Socket Sampling Mode=::NewEnumerator5
   )
  ]
  rendererClasses [&NiagaraMeshRendererProperties]
  renderers [{rendererClass=&NiagaraMeshRendererProperties, rendererIndex=0}]
 }
 NE_HUD {
  emitterUpdateScript [
   &EmitterState(Life Cycle Mode=::NewEnumerator1, Loop Duration=0.5, UseLoopDelay=false)
   &SpawnBurst_Instantaneous(
    Spawn Count=3
    Spawn Probability=HlslExpression() {hlslExpression="Emitter.LoopCount == 0"}
    Use Spawn Probability=true
    UseLoopCountLimit=false
   )
  ]
  particleSpawnScript [
   &FN_SetSpawnGroup(Number of Spawn Groups=3)
   /SetVariables(Particles.VisibilityTag=$Particles.SpawnGroup) {isSetParameters=true}
   &InitializeParticle(
    Color=&MakeCustomLinearColorFromBool(Bool=$User.bIsFrontEndPreview, True Color=<rgba 0.0589 0.625 0 1>, False Color=<rgba 0.1177 1.25 0 1>)
    Lifetime=0.25
    Mesh Scale=<0.5 0.5 1>
    Mesh Scale Mode=::NewEnumerator4
    Position Offset=&SelectVectorFromArray(
     Direct Array Index=$Particles.VisibilityTag
     Vector Selection Array=DataInterface() {propertyValues {FloatData [<-50 100 50>, <83 83 25>, <88 86 50>]}}
    )
    UsePositionOffset=true
    Write Material Random=true
   )
   &InitialMeshOrientation(
    Mesh Orientation Mode=::NewEnumerator2
    Orientation Vector=<0 0 0>
    Rotation=&MakeCustomVectorFromBool(
     Bool=$User.bIsFrontEndPreview
     True Vector=&SelectVectorFromArray(
      Direct Array Index=$Particles.VisibilityTag
      Vector Selection Array=DataInterface() {propertyValues {FloatData [<0 0 0.31>, <0 0 0.1>, <0 0 0.1>]}}
     )
     False Vector=&SelectVectorFromArray(
      Direct Array Index=$Particles.VisibilityTag
      Vector Selection Array=DataInterface() {propertyValues {FloatData [<-0.1 0.05 0.31>, <-0.05 0.05 0.1>, <-0.05 0.05 0.1>]}}
     )
    )
    Rotation Coordinate Space=::NewEnumerator2
    Use Rotation Vector=true
   )
  ]
  particleUpdateScript [
   &ParticleState(
    Kill Particles When Lifetime Has Elapsed=false
    Let Infinitely Lived Particles Die When Emitter Deactivates=true
    Loop Particles Lifetime=true
   )
   &SolveForcesAndVelocity(Clamp Velocity=false, Limit Acceleration=false)
   &DynamicMaterialParameters(
    Index 0 Param 1=&Lerp_Float(A=3, B=10, Alpha=$User.ParaGlideLeanAlpha)
    Index 0 Param 2=&MakeCustomFloatFromBool(Bool=$User.bIsFrontEndPreview, True Float=5, False Float=1)
    Module.Index0Float0Write=true
    Module.Index0Float1Write=true
    Module.Index0Float2Write=true
    Module.Index0Float3Write=true
   )
  ]
  rendererClasses [&NiagaraMeshRendererProperties]
  renderers [
   {rendererClass=&NiagaraMeshRendererProperties, rendererIndex=0}
   {rendererClass=&NiagaraMeshRendererProperties, rendererIndex=1}
   {rendererClass=&NiagaraMeshRendererProperties, rendererIndex=2}
  ]
 }
 NE_Laser {
  emitterUpdateScript [
   &EmitterState()
   &SpawnBurst_Instantaneous(
    Spawn Count=2
    Spawn Probability=HlslExpression() {hlslExpression="Emitter.LoopCount == 0"}
    Use Spawn Probability=true
    UseLoopCountLimit=false
   )
  ]
  particleSpawnScript [
   &PartitionParticles()
   &InitializeParticle(
    Color=<rgba 0.0912 6 0 1>
    Mesh Scale=<0.2 0.15 0.15>
    Mesh Scale Mode=::NewEnumerator4
    Sprite Size Mode=::NewEnumerator3
    Uniform Sprite Size=50
    UsePositionOffset=false
    Write Material Random=true
   )
  ]
  particleUpdateScript [
   &ParticleState(
    Kill Particles When Lifetime Has Elapsed=false
    Let Infinitely Lived Particles Die When Emitter Deactivates=true
    Loop Particles Lifetime=true
   )
   &ScaleMeshSize(Scale Factor=&Lerp_Vector(
     A=&RandomRangeVector(Minimum=<1.6 1 1>, Maximum=<1.6 1 1>)
     B=<1.8 1 1>
     Alpha=$User.ParaGlideLeanAlpha
    ))
   &SolveForcesAndVelocity(Clamp Velocity=false, Limit Acceleration=false)
   &SkeletalMeshLocation(
    Bone / Socket Index=$Particles.PartitionParticles.Partition
    MaskSpawnGroup=false
    Mesh Sampling Type=::NewEnumerator5
    Orientation Sampling=::NewEnumerator1
    Sampled Orientation Coordinate Space=::Local
    Sampled Position Offset=&SelectVectorFromArray(
     Direct Array Index=$Particles.PartitionParticles.Partition
     Vector Selection Array=DataInterface() {propertyValues {FloatData [<0 0 0>, <0 0 0>]}}
    )
    Sampled Position Offset Coordinate Space:ENiagaraSkelMeshTransforms=::NewEnumerator2
    Skeletal Mesh=$System.Skeletal Mesh
    Socket Sampling Mode=::NewEnumerator5
   )
   &UpdateMeshOrientation(
    Facing Interpolation=true
    Facing Interpolation Rate=20
    Facing Position:NiagaraPosition=&AddVectorToPosition(
     Position=&Lerp_Vector(A=Linked() {linkedVariable {name=System.Position, type=NiagaraPosition}}, B=<1 40 1>, Alpha=$User.ParaGlideLeanAlpha)
     Vector=<0 0 0>
    )
    Orientation Method=::NewEnumerator2
    Override Quat=false
    TransformFacingOrigin=false
    TransformFacingPosition=false
    TransformUp=false
    Up Direction:Vector3f=<0 0 0>
   )
   &Color(Color=&LerpLinearColors(StartColor=<rgba 0.0912 6 0 1>, EndColor=<rgba 6 0.374 0.18 1>, LerpFactor=$User.ParaGlideLeanAlpha))
  ]
  rendererClasses [&NiagaraMeshRendererProperties, &NiagaraSpriteRendererProperties]
  renderers [
   {rendererClass=&NiagaraMeshRendererProperties, rendererIndex=0}
   {rendererClass=&NiagaraSpriteRendererProperties, rendererIndex=1}
  ]
 }
}
summary {
 systemName=NS_Contrail_AshenMagnusSlant
 emitters [
  {emitterName=NE_BaseRadar, rendererClasses [&NiagaraSpriteRendererProperties]}
  {emitterName=NE_HUD, rendererClasses [&NiagaraMeshRendererProperties]}
  {emitterName=NE_Laser, rendererClasses [&NiagaraMeshRendererProperties, &NiagaraSpriteRendererProperties]}
  {emitterName=LWE_Targets, rendererClasses []}
  {emitterName=NE_Gadget, rendererClasses [&NiagaraMeshRendererProperties]}
  {emitterName=NE_Aim, rendererClasses [&NiagaraSpriteRendererProperties]}
 ]
 userVariables [
  User.bIsFrontEnd {type=NiagaraBool}
  User.bIsFrontEndPreview {type=NiagaraBool}
  User.ParaGlideLeanAlpha {type=NiagaraFloat, defaultValue=0, description=""}
  User.RibbonIDOffset {type=NiagaraInt32}
 ]
}
